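# Sound cue for the opening sound of a Seven Village door, played as an environment actor effect.
# One wave player feeds a Modulator at default settings for slight random variation, and every branch of the Quality Level root points to that same Modulator, so all quality levels sound alike.
# The cue uses the shared door attenuation and sound class, lowers its volume, and overrides concurrency to prevent new instances once the limit is reached.
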SoundCue /Game/Sounds/Artemis/Environment_Actors/Doors/Spot_Seven_Village_Door01_Open_Cue
root=QualityLevel_0
nodes [
 {id=Modulator_2, children [WavePlayer_6]}
 {id=QualityLevel_0, children [Modulator_2,Modulator_2,Modulator_2,Modulator_2]}
 {id=WavePlayer_6, properties {soundWaveAssetPtr=&/Spot_Seven_Village_Door01_Open}}
]
properties {
 attenuationSettings=/Game/Sounds/Attenuation/Fort_Doors_Att
 bOverrideConcurrency=true
 maxDistance=2400
 soundClassObject=&Fort_Doors
 volumeMultiplier=0.6
 concurrencyOverrides {resolutionRule=PreventNew}
}
